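# Sound cue for the reaction sound of the TechnoAwaken reactive back bling cosmetic.
# It picks one of three React wave variations at equal weight, applies slight random pitch and volume modulation, and plays after a short fixed delay.
# Routing uses the backbling sound class, a close pet attenuation and a reactive backbling audio property sheet, with reduced overall volume.
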SoundCue /BRCosmetics/Sounds/Backblings/TechnoAwaken/Backpack_TechnoAwaken_React_Cue
root=Delay_0
nodes [
 {id=Delay_0, children [Modulator_0], properties {delayMax=0.1, delayMin=0.1}}
 {id=Modulator_0, children [Random_0], properties {pitchMax=1.02, pitchMin=0.98, volumeMax=1.02, volumeMin=0.98}}
 {id=Random_0, children [WavePlayer_0,WavePlayer_1,WavePlayer_2], properties {weights [1,1,1]}}
 {id=WavePlayer_0, properties {soundWaveAssetPtr=&/Backpack_TechnoAwaken_React_01}}
 {id=WavePlayer_1, properties {soundWaveAssetPtr=&/Backpack_TechnoAwaken_React_02}}
 {id=WavePlayer_2, properties {soundWaveAssetPtr=&/Backpack_TechnoAwaken_React_03}}
]
properties {
 attenuationSettings=&Fort_Pets_Close_Athena_Att
 audioPropertiesSheet=/BRCosmetics/Sounds/Settings/PropertySheets/Reactive/APS_Cosmetic_Reactive_Backbling
 maxDistance=1800.0001
 soundClassObject=/Game/Sounds/FortSoundClassesAndModes/Fort_Backblings
 volumeMultiplier=0.45
}
